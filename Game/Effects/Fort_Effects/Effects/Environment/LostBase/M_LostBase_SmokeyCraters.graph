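# Translucent two-sided lit material for smoke plumes, apparently rising from craters in the LostBase environment.
# SplineThicken sets tapered strand widths, while world position offset adds a sine-animated bend weighted along UV.
# Two smoke textures pan with time, offset by vertex color and an actor-position seed.
# Opacity combines a vertex-green threshold driven by custom primitive data with length, fresnel, view-axis, near-camera and depth fades.
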Material /Game/Effects/Fort_Effects/Effects/Environment/LostBase/M_LostBase_SmokeyCraters
graph {
 properties {bAutomaticallySetUsageInEditor=false, blendMode=BLEND_Translucent, twoSided=true, usedShadingModels=Default Lit}
 nodes [
  n1 = NamedRerouteUsage()
  n2 = LinearInterpolate(B=ScalarParameter() {defaultValue=500, parameterName=Param_1}, Alpha=VertexColor().G) {constA=350}
  n3 = TextureCoordinate()
  n4 = OneMinus(None=Power(Base=ComponentMask(None=@n3) {g=true}, Exp=ScalarParameter() {defaultValue=2, parameterName=Param_10}))
  n5 = MaterialFunctionCall(WidthBase=Multiply(A=ScalarParameter() {defaultValue=75}, B=@n1), WidthTip=Multiply(A=Multiply(A=@n2, B=@n4), B=@n1)) {materialFunction=&SplineThicken}
  n6 = VectorParameter() {parameterName=Param_2, defaultValue=<rgba 375 375 75 0>}
  n7 = VectorParameter() {parameterName=Param_1, defaultValue=<rgba 350 350 50 0>}
  n8 = ScalarParameter() {defaultValue=0.1, parameterName=Param_7}
  n9 = Add(A=Time(), B=NamedRerouteUsage())
  n10 = Sine(None=Add(A=Multiply(A=@n8, B=@n9), B=VertexColor().G))
  n11 = OneMinus(None=ComponentMask(None=TextureCoordinate()) {g=true})
  n12 = ScalarParameter() {defaultValue=1.5, parameterName=Param_3}
  Add(A=@n5.WorldPositionOffset, B=LinearInterpolate(B=LinearInterpolate(A=@n6.RGB, B=@n7.RGB, Alpha=@n10), Alpha=Power(Base=@n11, Exp=@n12)))
  n13 = NamedRerouteUsage()
  n14 = Power(Base=OneMinus(None=Fresnel(Normal=@n13)), Exp=ScalarParameter() {defaultValue=4, parameterName=Param_5})
  n15 = ComponentMask(None=TextureCoordinate()) {g=true}
  n16 = ScalarParameter() {defaultValue=0.75, parameterName=Param_6}
  n17 = AppendVector(A=ScalarParameter() {defaultValue=1, parameterName=Param_8}, B=ScalarParameter() {defaultValue=1, parameterName=Param_9})
  n18 = TextureCoordinate()
  MaterialExpressionActorPositionWS_1 = ActorPositionWS()
  n19 = Convert(None=@MaterialExpressionActorPositionWS_1) {
   nodeName=Break Float3
   convertInputs [{
     type=Vector3
     expressionInput {
      expression=&self:MaterialExpressionActorPositionWS_1
      inputName=None
      mask=0
      maskA=0
      maskB=0
      maskG=0
      maskR=0
      outputIndex=0
     }
    }]
   convertOutputs [{type=Scalar}, {type=Scalar}, {type=Scalar}]
  }
  n20 = Add(A=@n19, B=@n19)
  n21 = Frac(None=Multiply(A=Add(A=@n20, B=@n19), B=ScalarParameter() {bUseCustomPrimitiveData=true, parameterName=Seed, primitiveDataIndex=1}) {constB=4565461})
  n22 = Add(A=VertexColor().R, B=Add(A=@n21, B=Time()))
  n23 = ScalarParameter() {defaultValue=0.25, parameterName=Param_2}
  n24 = MaterialFunctionCall(XY=Panner(Coordinate=Multiply(A=@n17, B=@n18), Time=Multiply(A=@n22, B=@n23)) {speedY=0.5}) {materialFunction=&Swizzle}
  n25 = TextureSample(UVs=@n24.YX) {samplerType=SAMPLERTYPE_Grayscale, texture=&T_Smoke_Trail_Soft}
  n26 = TextureSample(UVs=@n24.YX) {texture=/Game/Effects/Fort_Effects/Textures/Smoke/T_Puffy_Smoke_Trail_01}
  n27 = Multiply(
   A=Multiply(A=Power(Base=@n15, Exp=@n16), B=LinearInterpolate(A=@n25.G, B=@n26.G))
   B=MaterialFunctionCall(Distance=ScalarParameter() {defaultValue=700, parameterName=Param_4}) {materialFunction=&Near_Camera_Fade}.Result
  )
  n28 = Multiply(A=NamedRerouteUsage(), B=Multiply(A=Multiply(A=@n27), B=ScalarParameter() {defaultValue=1.5, parameterName=Overall Opacity}))
  n29 = Step(Y=ScalarParameter() {bUseCustomPrimitiveData=true, defaultValue=0.8114, parameterName=Step}, X=VertexColor().G)
  DepthFade(Opacity=Multiply(A=Multiply(A=@n14, B=@n28), B=@n29))
  LinearInterpolate(B=VectorParameter() {defaultValue=<rgba 0.0226 0.0226 0.0226 0>}.RGB, Alpha=ComponentMask(None=TextureCoordinate()) {g=true}) {constA=0.5}
  NamedRerouteDeclaration(None=@n5.Normal) {name=Normal, nodeColor=<rgba 0.6172 0 1 1>}
  NamedRerouteDeclaration(None=@n21) {name=Random, nodeColor=<rgba 1 0 0.4453 1>}
  n30 = VectorParameter() {parameterName=Axis, defaultValue=<rgba 0.12 0.12 0.12 0>}
  n31 = Saturate(None=DotProduct(
    A=MaterialFunctionCall!(Use World Position Offsets) {materialFunction=/Engine/Functions/Engine_MaterialFunctions02/Utility/CameraVectorWithWPOOptions}.Result
    B=Multiply(A=@n30.RGB)
   ))
  NamedRerouteDeclaration(None=Power(Base=OneMinus(None=@n31), Exp=ScalarParameter() {defaultValue=2, parameterName=Axis Power 2})) {name=Axis Fade, nodeColor=<rgba 0.2891 1 0 1>}
  NamedRerouteDeclaration(None=@n29) {name=Step, nodeColor=<rgba 0 1 0.7422 1>}
 ]
}
